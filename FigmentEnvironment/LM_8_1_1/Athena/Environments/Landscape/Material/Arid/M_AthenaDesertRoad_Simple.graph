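# Simplified desert road surface material for the arid landscape, usable on spline meshes.
# It builds on the simple Arid Sand function with a constant specular override, then blends in snow using the holiday snowfall mask.
# A UseDesertRoad switch passes that blend through the desert road function.
# A UseTopo switch multiplies base color by a terrain-coordinate color map, scaled by a landscape parameter collection value and faded out under snow.
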Material /FigmentEnvironment/LM_8_1_1/Athena/Environments/Landscape/Material/Arid/M_AthenaDesertRoad_Simple
graph {
 properties {bUseMaterialAttributes=true, bUsedWithSplineMeshes=true, pixelDepthOffsetMode=PDOM_Legacy, usedShadingModels=Default Lit}
 nodes [
  n1 = Abs(None=CollectionParameter() {collection=/FigmentEnvironment/LM_8_1_1/Athena/Environments/Landscape/MPC/MPC_Landscape, parameterName=Arid_Sand_TexSize})
  ComponentMask(None=Divide(A=WorldPosition() {worldPositionShaderOffset=WPT_ExcludeAllShaderOffsets}.XYZ, B=Multiply(A=@n1) {constB=-1})) {g=true, r=true}
  n2 = StaticBool()
  n3 = MaterialFunctionCall!(
   Use_ExtraSnowMask
   UV_WorldMask
   SnowStartBiasOverride=Constant()
   UseHeightLerp
   LandscapeBlend_Height
   Landscape_RockMask
   Snow_Mask
   Simple_UseWorldZNormal
   Simple_WorldNormal
   WorldPosition
  ) {materialFunction=/FigmentEnvironment/LM_8_1_1/Athena/Environments/Landscape/MaterialFunctions/Snow/MF_Athena_Holiday_SnowFall}
  n4 = MaterialFunctionCall!(
   Base Material=SetMaterialAttributes!(
    MaterialAttributes=MaterialFunctionCall!(In, UseRoughness=@n2) {materialFunction=/FigmentEnvironment/LM_8_1_1/Athena/Environments/Landscape/MaterialFunctions/Arid/MF_Athena_Arid_Sand}.Simple_Material
    Specular=Constant() {desc="Landscape Arid override spec value was 0.1.  Could be become a parameter", r=0.5}
   ) {attributeSetTypes [#1]}
   Top Material=MaterialFunctionCall!(UseLandscapeCoordsOnSwitch) {materialFunction=/FigmentEnvironment/LM_8_1_1/Athena/Environments/Landscape/MaterialFunctions/Snow/MF_Athena_Snow_01}.Material
   Alpha=@n3.SnowMask
  ) {materialFunction=/Engine/Functions/MaterialLayerFunctions/MatLayerBlend_Standard}
  n5 = StaticSwitchParameter(
   True=MaterialFunctionCall!(In=@n4.Blended Material) {materialFunction=/FigmentEnvironment/LM_8_1_1/Athena/Environments/Landscape/MaterialFunctions/Arid/MF_Athena_DesertRoad}.Result
   False=@n4.Blended Material
  ) {defaultValue=true, parameterName=UseDesertRoad}
  MaterialFunctionCall!(Material=@n5) {materialFunction=/Engine/Functions/MaterialLayerFunctions/MatLayerBlend_BreakBaseColor}
  n6 = GetMaterialAttributes(None=@n5) {attributeGetTypes [#2, #1]}
  n7 = TextureSampleParameter2D(UVs=MaterialFunctionCall!(WorldPosition) {materialFunction=/FigmentEnvironment/LM_8_1_1/Athena/Environments/Landscape/MaterialFunctions/Utilities/MF_Athena_Terrain_Coords}.Result) {
   parameterName=CombinedColorMask
   samplerType=SAMPLERTYPE_LinearColor
   texture=/FigmentEnvironment/LM_8_1_1/Athena/Environments/Landscape/Textures/T_Athena_Terrain_CombinedColors_01
  }
  n8 = LinearInterpolate(
   A=Multiply(
    A=@n7.RGB
    B=CollectionParameter() {collection=&MPC_Landscape, parameterName=ColormapDivisor}
   )
   Alpha=@n3.SnowMask
  )
  StaticSwitchParameter(
   True=SetMaterialAttributes!(MaterialAttributes=@n6.MaterialAttributes, Base Color=Multiply(A=@n6.Base Color, B=@n8)) {attributeSetTypes [#2]}
   False=@n5
  ) {defaultValue=true, parameterName=UseTopo}
 ]
}
